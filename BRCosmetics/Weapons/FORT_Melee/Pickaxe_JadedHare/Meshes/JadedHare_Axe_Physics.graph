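# Physics asset for the JadedHare pickaxe, a two-handed melee cosmetic with a dangling chain.
# Capsule bodies cover the handle, root bone, chain base and six dynamic chain bones.
# The handle and chain base have all angular motion locked to the root.
# The chain links connect in series with swing limits and tight twist limits, and the last two joints swing less than those above.
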PhysicsAsset /BRCosmetics/Weapons/FORT_Melee/Pickaxe_JadedHare/Meshes/JadedHare_Axe_Physics
bodies {
 Melee_TwoHanded_Handle [{
   shapeType=Capsule
   length=20.703
   radius=3.2662
   shapeName=Melee_TwoHanded_Handle_capsule
   center=<0.0018 -1.0249 5.7661>
   rotation=<rot -3.1004 -90 0.0155>
  }]
 b_AR_Root [{shapeType=Capsule, length=4.7772, radius=6.7838, shapeName=b_AR_Root_capsule, center=<0.0441 0 -1.6618>, rotation=<rot -0.0042 -90 89.3914>}]
 base_Chain_01 [{
   shapeType=Capsule
   length=1.3013
   radius=0.5751
   shapeName=base_Chain_01_capsule
   center=<0.2122 0 0.0158>
   rotation=<rot 0 -89.9998 38.3673>
  }]
 dyn_Chain_01 [{shapeType=Capsule, length=0.51, radius=0.5, shapeName=dyn_Chain_01_capsule, center=<1.6718 0 -0.0142>, rotation=<rot 0 90 -88.7759>}]
 dyn_Chain_02 [{shapeType=Capsule, length=0.505, radius=0.505, shapeName=dyn_Chain_02_capsule, center=<1.213 0 0.0125>, rotation=<rot 0 90 -88.7747>}]
 dyn_Chain_03 [{shapeType=Capsule, length=0.505, radius=0.505, shapeName=dyn_Chain_03_capsule, center=<1.0013 0 0.0499>, rotation=<rot 0 90 -90>}]
 dyn_Chain_04 [{shapeType=Capsule, length=1.4619, radius=1.3524, shapeName=dyn_Chain_04_capsule, center=<2.1735 0.0027 0.1218>, rotation=<rot 0 90 -90>}]
 dyn_Chain_05 [{shapeType=Capsule, length=1.636, radius=1.3329, shapeName=dyn_Chain_05_capsule, center=<1.092 0 0>, rotation=<rot 0 90 -90>}]
 dyn_Chain_06 [{shapeType=Capsule, length=1.7752, radius=0.8706, shapeName=dyn_Chain_06_capsule, center=<0.287 -0.0249 -0.0021>, rotation=<rot 0 90 -90>}]
}
constraints [
 {bone1Name=dyn_Chain_01, bone2Name=base_Chain_01, swing1LimitDegrees=40, swing2LimitDegrees=80, twistLimitDegrees=5}
 {bone1Name=dyn_Chain_02, bone2Name=dyn_Chain_01, swing1LimitDegrees=40, swing2LimitDegrees=40, twistLimitDegrees=5}
 {bone1Name=dyn_Chain_03, bone2Name=dyn_Chain_02, swing1LimitDegrees=40, swing2LimitDegrees=40, twistLimitDegrees=5}
 {bone1Name=dyn_Chain_04, bone2Name=dyn_Chain_03, swing1LimitDegrees=40, swing2LimitDegrees=40, twistLimitDegrees=5}
 {bone1Name=dyn_Chain_05, bone2Name=dyn_Chain_04, swing1LimitDegrees=20, swing2LimitDegrees=20, twistLimitDegrees=5}
 {bone1Name=dyn_Chain_06, bone2Name=dyn_Chain_05, swing1LimitDegrees=20, swing2LimitDegrees=20, twistLimitDegrees=5}
 {
  bone1Name=Melee_TwoHanded_Handle
  bone2Name=b_AR_Root
  swing1LimitDegrees=45
  swing1Motion=Locked
  swing2LimitDegrees=45
  swing2Motion=Locked
  twistLimitDegrees=45
  twistMotion=Locked
 }
 {
  bone1Name=base_Chain_01
  bone2Name=b_AR_Root
  swing1LimitDegrees=45
  swing1Motion=Locked
  swing2LimitDegrees=45
  swing2Motion=Locked
  twistLimitDegrees=45
  twistMotion=Locked
 }
]
